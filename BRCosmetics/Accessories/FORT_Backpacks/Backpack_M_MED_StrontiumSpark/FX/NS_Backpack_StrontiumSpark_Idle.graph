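# Niagara idle effect for the StrontiumSpark backpack cosmetic that shows a flickering red eye with a glow.
# Eye and Glow emitters each burst-spawn sprites that persist past their lifetime, with Eye using a position offset and fixed alignment and Glow a camera offset.
# Spawn color is tinted by time of day.
# Irregular alpha curves, indexed by the system's normalized loop age, drive a separate flicker for each layer.
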NiagaraSystem /BRCosmetics/Accessories/FORT_Backpacks/Backpack_M_MED_StrontiumSpark/FX/NS_Backpack_StrontiumSpark_Idle
dependencies {usedDataInterfaces [&NiagaraDataInterfaceCurve], usedDynamicInputs [], usedModules [&SystemState]}
emitters {
 Eye {
  emitterUpdateScript [
   &EmitterState(Life Cycle Mode=::NewEnumerator1, Loop Behavior=::NewEnumerator1, Loop Duration=1, UseLoopDelay=false)
   &SpawnBurst_Instantaneous(Use Spawn Probability=false, UseLoopCountLimit=false)
  ]
  particleSpawnScript [
   &InitializeParticle(
    Color=<rgba 3 0.5227 0.4032 1>
    Position Offset=<0.8 0 0.18>
    Sprite Size Mode=::NewEnumerator3
    Uniform Sprite Size=20
    UsePositionOffset=true
    Write Material Random=true
   )
   &Color(Color=&MultiplyLinearColor(
     Linear Color=$Particles.Color
     Scale Factor=&MakeVector4_FromLinearColor(LinearColor=&FNTimeOfDayColor(
       Night Color=<rgba 0.6795 0.686 0.822 0.8>
       Morning Color=<rgba 0.7 0.6119 0.6049 0.8>
       Day Color=<rgba 1 1 1 2>
       Evening Color=<rgba 1 1 1 2>
      ))
    ))
   /SetVariables(Particles.SpriteAlignment=<0 0 1>) {isSetParameters=true}
  ]
  particleUpdateScript [
   &ParticleState(Kill Particles When Lifetime Has Elapsed=false)
   &ScaleColor(
    Scale Alpha=&FloatFromCurve(
     FloatCurve=DataInterface() {propertyValues {Curve {keys [<at 0 1>,<at 0.1 0.15>,<at 0.15 0.8>,<at 0.2 0.4>,<at 0.25 1>,<at 0.3 0.5>,<at 0.4 0.5>,<at 0.43 0.3>,<at 0.5 0.7>,<at 0.55 0.1>,<at 0.6 0.5>,<at 0.7 0.25>,<at 0.8 0.85>,<at 1 0.9>]}}}
     CurveIndex=$System.NormalizedLoopAge
    )
    ScaleA=true
    ScaleRGB=true
   )
   &CameraOffset()
  ]
  rendererClasses [&NiagaraSpriteRendererProperties]
  renderers [{rendererClass=&NiagaraSpriteRendererProperties}]
 }
 Glow {
  emitterUpdateScript [
   &EmitterState(Life Cycle Mode=::NewEnumerator1, Loop Behavior=::NewEnumerator1, Loop Duration=1, UseLoopDelay=false)
   &SpawnBurst_Instantaneous(Use Spawn Probability=false, UseLoopCountLimit=false)
  ]
  particleSpawnScript [
   &InitializeParticle(Color=<rgba 2 0 0.0316 0.4>, Sprite Size Mode=::NewEnumerator3, Uniform Sprite Size=11, UsePositionOffset=false, Write Material Random=true)
   &Color(Color=&MultiplyLinearColor(
     Linear Color=$Particles.Color
     Scale Factor=&MakeVector4_FromLinearColor(LinearColor=&FNTimeOfDayColor(
       Night Color=<rgba 0.6795 0.686 0.822 0.8>
       Morning Color=<rgba 0.7 0.6119 0.6049 0.8>
       Day Color=<rgba 1 1 1 1.1>
       Evening Color=<rgba 1 1 1 1.1>
      ))
    ))
  ]
  particleUpdateScript [
   &ParticleState(Kill Particles When Lifetime Has Elapsed=false)
   &ScaleColor(
    Scale Alpha=&FloatFromCurve(
     FloatCurve=DataInterface() {propertyValues {Curve {keys [<at 0 1>,<at 0.1 0.5>,<at 0.2 1>,<at 0.3 0.55>,<at 0.4 0.55>,<at 0.5 0.1>,<at 0.6 0.5>,<at 0.7 0.15>,<at 0.8 0.55>,<at 0.9 0.55>,<at 1 0.9>]}}}
     CurveIndex=$System.NormalizedLoopAge
    )
    ScaleA=true
    ScaleRGB=true
   )
   &CameraOffset(Camera Offset Amount=2)
  ]
  rendererClasses [&NiagaraSpriteRendererProperties]
  renderers [{rendererClass=&NiagaraSpriteRendererProperties}]
 }
}
summary {
 systemName=NS_Backpack_StrontiumSpark_Idle
 emitters [
  {emitterName=Glow, rendererClasses [&NiagaraSpriteRendererProperties]}
  {emitterName=Eye, rendererClasses [&NiagaraSpriteRendererProperties]}
 ]
 userVariables []
}
