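# Niagara effect for the StrontiumSparks backpack's elimination reaction, showing an animated red hologram face.
# Four burst emitters each play over about five seconds: a flickering Glow sprite with a particle light, a blinking, darting HologramEye, layered HologramShapes and a growing mesh Lines element.
# HologramShapes drives four sprite renderers from one emitter, using separate size, rotation and randomized material parameters for circles, border and arrow.
# Sprite colors are scaled by time of day.
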NiagaraSystem /BRCosmetics/Accessories/FORT_Backpacks/Backpack_M_MED_StrontiumSpark/FX/NS_Backpack_StrontiumSparks_Elimination
dependencies {
 usedDataInterfaces [
  &NiagaraDataInterfaceCurve
  &NiagaraDataInterfaceVector2DCurve
  &NiagaraDataInterfaceVectorCurve
 ]
 usedDynamicInputs []
 usedModules [&SystemState]
}
emitters {
 Glow {
  emitterUpdateScript [
   &EmitterState()
   &SpawnBurst_Instantaneous(Use Spawn Probability=false, UseLoopCountLimit=false)
  ]
  particleSpawnScript [
   &InitializeParticle(
    Color=<rgba 1 0 0.0495 0.12>
    Lifetime=5.5
    Position Offset=<4 0 0.18>
    Sprite Size Mode=::NewEnumerator3
    Uniform Sprite Size=90
    UsePositionOffset=true
    Write Material Random=true
   )
   &Color(Color=&MultiplyLinearColor(
     Linear Color=$Particles.Color
     Scale Factor=&MakeVector4_FromLinearColor(LinearColor=&FNTimeOfDayColor(
       Night Color=<rgba 0.6795 0.686 0.822 0.8>
       Morning Color=<rgba 0.7 0.6119 0.6049 0.8>
       Day Color=<rgba 1 1 1 1>
       Evening Color=<rgba 1 1 1 1>
      ))
    ))
   /SetVariables(Particles.SpriteAlignment=<0 0 1>) {isSetParameters=true, module=SetVariables_#1}
  ]
  particleUpdateScript [
   &ParticleState()
   &ScaleSpriteSize(Uniform Curve Sprite Scale=DataInterface() {propertyValues {Curve {keys [{tangentMode=RCTM_User, time=0, value=0}, <at 0.4 1>, <at 0.9 1>, <at 0.95 1.3>, {tangentMode=RCTM_User, time=1, value=0}]}}})
   &CameraOffset(Camera Offset Amount=4)
   &ScaleColor(
    Scale Alpha=&FloatFromCurve(FloatCurve=DataInterface() {propertyValues {Curve {keys [<at 0 0.1>,<at 0.06 1>,<at 0.1 0.2>,<at 0.2 0.85>,<at 0.3 0>,<at 0.4 0.7>,<at 0.5 0.1>,<at 0.6 0.95>,<at 0.7 0.3>,<at 0.8 1>,<at 1 0>]}}})
    ScaleA=true
    ScaleRGB=true
   ) {module=ScaleColor001}
   &DynamicMaterialParameters(
    Index 0 Param 1=1
    Module.Index0Param0Write=true
    Module.Index0Param1Write=true
    Module.Index0Param2Write=true
    Module.Index0Param3Write=true
   )
   /SetVariables(Particles.SpriteRotation:NiagaraFloat=&FloatFromCurve(
     FloatCurve=DataInterface() {propertyValues {Curve {keys [<at 0.2 0>,<at 0.3 0.35>,<at 0.4 0.35>,<at 0.5 -0.4>,<at 0.6 0.1>,<at 0.7 0.1>,<at 0.8 -0.45>,<at 1 0>]}}}
     Scale Curve=180
    )) {isSetParameters=true, module=SetVariables_#2}
   &FN_Light_v2(
    Light Exponent=2
    Light Radius:NiagaraFloat=&FloatFromCurve(
     FloatCurve=DataInterface() {propertyValues {Curve {keys [<at 0 0>,<at 0.12 0.4>,<at 0.14 0.4>,<at 0.16 0.65>,<at 0.25 1>,<at 0.9 1>,<at 1 0.3>]}}}
     Scale Curve=50
    )
    LightEnabled=true
    Position Offset:Vector3f=<-5 0 0>
    Scale Color:NiagaraFloat=&FloatFromCurve(
     FloatCurve=DataInterface() {propertyValues {Curve {keys [<at 0 0>,<at 0.15 1>,<at 0.16 0>,<at 0.2 0.7>,<at 0.24 0.15>,<at 0.3 1>,<at 0.33 0.35>,<at 0.36 1>,<at 0.4 0>,<at 0.43 0.75>,<at 0.45 0>,<at 0.5 1>,<at 0.53 0>,<at 0.56 0.8>,<at 0.6 0>,<at 0.63 0.9>,<at 0.65 0.2>,<at 0.7 1>,<at 0.75 0>,<at 0.8 0.65>,<at 0.9 0.1>,<at 0.95 1>,<at 1 0>,<at 1.5 0>]}}}
     Scale Curve=8
    )
   )
  ]
  rendererClasses [&NiagaraSpriteRendererProperties, &NiagaraLightRendererProperties]
  renderers [
   {rendererClass=&NiagaraSpriteRendererProperties}
   {rendererClass=&NiagaraLightRendererProperties, rendererIndex=1}
  ]
 }
 HologramEye {
  emitterUpdateScript [
   &EmitterState()
   &SpawnBurst_Instantaneous(Use Spawn Probability=false, UseLoopCountLimit=false)
  ]
  particleSpawnScript [
   &InitializeParticle(
    Lifetime=5.5
    Position Offset=<3 0 0>
    Sprite Size Mode=::NewEnumerator3
    Uniform Sprite Size=40
    UsePositionOffset=true
    Write Material Random=true
   )
   &Color(Color=&MultiplyLinearColor(
     Linear Color=$Particles.Color
     Scale Factor=&MakeVector4_FromLinearColor(LinearColor=&FNTimeOfDayColor(
       Night Color=<rgba 0.6795 0.686 0.822 0.8>
       Morning Color=<rgba 0.7 0.6119 0.6049 0.8>
       Day Color=<rgba 1 1 1 2>
       Evening Color=<rgba 1 1 1 2>
      ))
    ))
   /SetVariables(Particles.SpriteAlignment=<0 0 1>) {isSetParameters=true, module=SetVariables_#1}
  ]
  particleUpdateScript [
   &ParticleState()
   &ScaleSpriteSize(Uniform Curve Sprite Scale=DataInterface() {propertyValues {Curve {keys [{tangentMode=RCTM_User, time=0.2, value=0}, <at 0.3 1>, <at 0.9 1>, <at 0.95 1.3>, {tangentMode=RCTM_User, time=1, value=0}]}}})
   &ScaleSpriteSize(
    Non-Uniform Curve Sprite Scale=DataInterface() {propertyValues {
      XCurve {keys [{tangentMode=RCTM_User, time=0, value=1}]}
      YCurve {keys [
        {tangentMode=RCTM_User, time=0.2, value=1}
        <at 0.22 0>
        <at 0.25 1>
        <at 0.35 1>
        <at 0.37 0>
        <at 0.4 1>
        <at 0.43 0>
        <at 0.46 1>
        <at 0.7 1>
        <at 0.73 0>
        <at 0.76 1>
        <at 1 1>
       ]}
     }}
    Scale Sprite Size Mode=::NewEnumerator3
   ) {module=ScaleSpriteSize001}
   /SetVariables(Particles.Position=&AddVectorToPosition(
     Position=$Particles.Initial.Position
     Vector=&VectorFromCurve(
      VectorCurve=DataInterface() {propertyValues {
        XCurve {keys [<at 0 1>,<at 1 1>]}
        YCurve {keys [<at 0 0>,<at 0.15 0>,<at 0.2 1>,<at 0.25 1>,<at 0.28 -1>,<at 0.4 -1>,<at 0.45 -0.3>,<at 0.5 -0.3>,<at 0.55 0.6>,<at 0.6 0>,<at 1 0>]}
        ZCurve {keys [<at 0 0>,<at 0.31 0>,<at 0.35 1.025>,<at 0.4 1>,<at 0.44 -0.4>,<at 0.53 -0.4>,<at 0.6 0.7>,<at 0.7 -1>,<at 0.8 -1>,<at 0.84 1>,<at 0.9 0>,<at 1 0>]}
       }}
      Scale Curve=<1 1.5 1.6>
     )
    )) {isSetParameters=true}
   &CameraOffset(Camera Offset Amount=4)
   &ScaleColor(
    Scale Alpha=&FloatFromCurve(FloatCurve=DataInterface() {propertyValues {Curve {keys [<at 0 0.1>,<at 0.06 1>,<at 0.8 1>,<at 1 0>]}}})
    ScaleA=true
    ScaleRGB=true
   ) {module=ScaleColor001}
   &DynamicMaterialParameters(
    Index 0 Param 1=1
    Module.Index0Param0Write=true
    Module.Index0Param1Write=true
    Module.Index0Param2Write=true
    Module.Index0Param3Write=true
   )
  ]
  rendererClasses [&NiagaraSpriteRendererProperties]
  renderers [{rendererClass=&NiagaraSpriteRendererProperties}]
 }
 HologramShapes {
  emitterUpdateScript [
   &EmitterState()
   &SpawnBurst_Instantaneous(Use Spawn Probability=false, UseLoopCountLimit=false)
  ]
  particleSpawnScript [
   &InitializeParticle(
    Color=<rgba 1 0 0.0495 1>
    Lifetime=5.5
    Position Offset=<4 0 0>
    Sprite Size Mode=::NewEnumerator3
    Uniform Sprite Size=50
    UsePositionOffset=true
    Write Material Random=true
   )
   &Color(Color=&MultiplyLinearColor(
     Linear Color=$Particles.Color
     Scale Factor=&MakeVector4_FromLinearColor(LinearColor=&FNTimeOfDayColor(
       Night Color=<rgba 0.6795 0.686 0.822 0.8>
       Morning Color=<rgba 0.7 0.6119 0.6049 0.8>
       Day Color=<rgba 1 1 1 2>
       Evening Color=<rgba 1 1 1 2>
      ))
    ))
   &DynamicMaterialParameters(
    Index 0 Param 1=&RandomRangeFloat(Minimum=0, Maximum=2)
    Index 0 Param 2=&RandomRangeFloat(Minimum=0, Maximum=1)
    Index 0 Param 3=&RandomRangeFloat(Minimum=0.25, Maximum=1)
    Module.Index0Param0Write=true
    Module.Index0Param1Write=true
    Module.Index0Param2Write=true
    Module.Index0Param3Write=true
   )
   /SetVariables(Particles.SpriteAlignment=<0 0 1>) {isSetParameters=true, module=SetVariables_#1}
   /SetVariables(Particles.BorderSize:Vector2f=&Vector2DFromFloat(Value=53)) {isSetParameters=true}
   /SetVariables(
    Particles.ArrowDynParam:Vector4f=&MakeVector4(
     X=&RandomRangeFloat(Minimum=0.25, Maximum=2)
     Y=&RandomRangeFloat(Minimum=0.25, Maximum=2)
     Z=&RandomRangeFloat(Minimum=0.25, Maximum=1)
     W=0
    )
    Particles.BorderDynParam:Vector4f=&MakeVector4(
     X=&RandomRangeFloat(Minimum=0.25, Maximum=1)
     Y=&RandomRangeFloat(Minimum=0.25, Maximum=1)
     Z=&RandomRangeFloat(Minimum=0.2, Maximum=1)
     W=0
    )
    Particles.Circle2DynParam:Vector4f=&MakeVector4(
     X=&RandomRangeFloat(Minimum=0.25, Maximum=2)
     Y=&RandomRangeFloat(Minimum=0.25, Maximum=1)
     Z=&RandomRangeFloat(Minimum=0.2, Maximum=1)
     W=0
    )
   ) {isSetParameters=true}
  ]
  particleUpdateScript [
   &ParticleState()
   &ScaleSpriteSize(Uniform Curve Sprite Scale=DataInterface() {propertyValues {Curve {keys [{tangentMode=RCTM_User, time=0, value=0}, <at 0.05 1.12>, <at 0.1 0.8>, <at 0.15 1>, <at 0.96 1>, <at 1 0>]}}})
   &CameraOffset(Camera Offset Amount=4)
   &ScaleColor(
    Scale Alpha=&FloatFromCurve(
     FloatCurve=DataInterface() {propertyValues {Curve {keys [<at 0 0.1>,<at 0.06 1>,<at 0.8 1>,<at 1 0.5>]}}}
     CurveIndex=$System.NormalizedLoopAge
    )
    ScaleA=true
    ScaleRGB=true
   ) {module=ScaleColor001}
   /SetVariables(Particles.SpriteRotation:NiagaraFloat=&FloatFromCurve(
     FloatCurve=DataInterface() {propertyValues {Curve {keys [<at 0 0>,<at 0.1 0.775>,<at 0.2 0>,<at 0.4 0.35>,<at 0.5 -0.4>,<at 0.6 0.1>,<at 0.7 0.1>,<at 0.8 -0.45>,<at 1 0>]}}}
     Scale Curve=&RandomRangeFloat(Minimum=100, Maximum=180)
    )) {isSetParameters=true, module=SetVariables_#2}
   /SetVariables(
    Particles.Circle2Rotation:NiagaraFloat=&FloatFromCurve(
     FloatCurve=DataInterface() {propertyValues {Curve {keys [<at 0.2 0>,<at 0.3 0.35>,<at 0.4 0.35>,<at 0.5 -0.4>,<at 0.6 0.1>,<at 0.7 0.1>,<at 0.8 -0.45>,<at 1 0>]}}}
     Scale Curve=&RandomRangeFloat(Minimum=-30, Maximum=-60)
    )
    Particles.Circle2Size:Vector2f=&Multiply_Vector2D_ByFloat(
     Float=&FloatFromCurve(FloatCurve=DataInterface() {propertyValues {Curve {keys [<at 0.08 0>, {tangentMode=RCTM_User, time=0.09, value=0.45}, <at 0.12 1.12>, <at 0.14 0.8>, <at 0.18 1>, <at 0.96 1>, <at 1 0>]}}})
     Vector2D=$Particles.Initial.SpriteSize
    )
   ) {isSetParameters=true}
   /SetVariables(
    Particles.BorderRotation:NiagaraFloat=&Sine(Normalized Angle=$Emitter.Age, Period=2, Scale=&RandomRangeFloat(Minimum=15, Maximum=30), Bias=0)
    Particles.BorderSize:Vector2f=&Multiply_Vector2D_ByFloat(
     Float=&FloatFromCurve(FloatCurve=DataInterface() {propertyValues {Curve {keys [<at 0.14 0>, {tangentMode=RCTM_User, time=0.15, value=0.7}, <at 0.18 1.12>, <at 0.2 0.8>, <at 0.25 1>, <at 0.96 1>, <at 1 0>]}}})
     Vector2D=$Particles.Initial.BorderSize
    )
   ) {isSetParameters=true}
   /SetVariables(
    Particles.ArrowRotation:NiagaraFloat=&Sine(Normalized Angle=$Emitter.Age, Period=2, Scale=5, Bias=0)
    Particles.ArrowSize:Vector2f=&Multiply_Vector2D(
     A=&Vector2DFromCurve(
      Vector2Curve=DataInterface() {propertyValues {
        XCurve {keys [
          <at 0.2 0>
          {tangentMode=RCTM_User, time=0.21, value=0.45}
          {tangentMode=RCTM_User, time=0.25, value=1.1}
          <at 0.4 1>
          <at 0.5 1.1>
          <at 0.6 1>
          <at 0.7 1.1>
          <at 0.8 1>
          <at 0.9 1.1>
          <at 0.96 1.1>
          <at 1 0>
         ]}
        YCurve {keys [<at 0.2 0>, {tangentMode=RCTM_User, time=0.21, value=0.45}, {tangentMode=RCTM_User, time=0.25, value=1}]}
       }}
      ScaleCurve=<1 1>
     )
     B=$Particles.Initial.SpriteSize
    )
   ) {isSetParameters=true}
  ]
  rendererClasses [&NiagaraSpriteRendererProperties]
  renderers [
   {rendererClass=&NiagaraSpriteRendererProperties}
   {rendererClass=&NiagaraSpriteRendererProperties, rendererIndex=1}
   {rendererClass=&NiagaraSpriteRendererProperties, rendererIndex=2}
   {rendererClass=&NiagaraSpriteRendererProperties, rendererIndex=3}
  ]
 }
 Lines {
  emitterUpdateScript [
   &EmitterState()
   &SpawnBurst_Instantaneous(Use Spawn Probability=false, UseLoopCountLimit=false)
  ]
  particleSpawnScript [&InitializeParticle(
    Color=<rgba 1 1 1 0.1>
    Lifetime=5
    Mesh Scale=<0.37 0.9 0.8>
    Mesh Scale Mode=::NewEnumerator4
    Position Offset=<-5 0 0>
    UsePositionOffset=true
    Write Material Random=true
   )]
  particleUpdateScript [
   &ParticleState()
   &ScaleMeshSize(Scale Factor=&VectorFromCurve(VectorCurve=DataInterface() {propertyValues {XCurve {keys [<at 0 1>,<at 1 1>]}, YCurve {keys [<at 0 0.2>,<at 0.4 1>]}, ZCurve {keys [<at 0 0.2>,<at 0.4 1>]}}}))
  ]
  rendererClasses [&NiagaraMeshRendererProperties]
  renderers [{rendererClass=&NiagaraMeshRendererProperties, rendererIndex=0}]
 }
}
summary {
 systemName=NS_Backpack_StrontiumSparks_Elimination
 emitters [
  {emitterName=HologramEye, rendererClasses [&NiagaraSpriteRendererProperties]}
  {emitterName=Glow, rendererClasses [&NiagaraSpriteRendererProperties, &NiagaraLightRendererProperties]}
  {emitterName=Lines, rendererClasses [&NiagaraMeshRendererProperties]}
  {emitterName=HologramShapes, rendererClasses [&NiagaraSpriteRendererProperties]}
 ]
 userVariables [
  User.bIsFrontEndPreview {type=NiagaraBool}
  User.DropShadowColor {type=LinearColor, description="", defaultValue {struct=LinearColor, value=<rgba 2 0.2057 0 1>}}
  User.ShapeColor {type=LinearColor, description="", defaultValue {struct=LinearColor, value=<rgba 1.5 0 0 1>}}
  User.SkeletalMesh {
   type=NiagaraDataInterfaceSkeletalMesh
   defaultValue {
    struct=NiagaraExt_VariableValue_DataInterface
    value {dataInterface=&self:NiagaraDataInterfaceSkeletalMesh_0}
   }
  }
 ]
}
