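# Anim Blueprint for the StrontiumSpark backpack cosmetic, animating the pack and its floating sphere.
# A cached state machine switches between randomized idles (varied for front end), skydiving, and an elimination reaction triggered by an anim input event.
# The output adds a time-driven random vector offset to dyn_sphereFloat.
# The Anim Dynamics/spring skydive blend and Animation Testing machine do not feed the output.
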AnimBlueprint /BRCosmetics/Accessories/FORT_Backpacks/Backpack_M_MED_StrontiumSpark/Meshes/M_MED_StrontiumSpark_Pack_AnimBP
target_skeleton=&/M_MED_StrontiumSpark_Pack_Skeleton
graphs [
 AnimGraph {nodes [
   n1 = Anim Dynamics - Bone: MSTR_sphereFloat(ComponentPose=Local To Component(LocalPose=Use cached pose 'GroundPose'()), Alpha=Get Skydive() -> (Skydive:double), ExternalForce="0,0,0") {
    angularDampingOverride=0.999
    angularSpringConstant=20000
    bAngularSpring=true
    bGravityOverrideInSimSpace=true
    bLinearSpring=true
    bOverrideAngularDamping=true
    bOverrideLinearDamping=true
    bUseGravityOverride=true
    gravityScale=0.35
    linearDampingOverride=0.99
    linearSpringConstant=150
    simulationSpace=RootRelative
    alphaBoolBlend {blendOption=Linear}
    alphaScaleBiasClamp {bClampResult=true, clampMax=0.2}
    boundBone {boneName=MSTR_sphereFloat}
    componentAppliedLinearAccClamp=<5000 5000 5000>
    componentLinearAccScale=<0.3 0.3 0.75>
    componentLinearVelScale=<0.3 0.3 0.75>
    gravityOverride {z=-980}
    physicsBodyDefinitions [{
      sphereCollisionRadius=0
      boundBone {boneName=MSTR_sphereFloat}
      boxExtents=<20 30 30>
      constraintSetup {angularLimitsMax=<0 0 0>, angularLimitsMin=<0 0 0>, angularTarget=<1 0 0>, linearAxesMax=<0 50 100>, linearAxesMin=<-50 -50 -50>}
      localJointOffset=<0 0 0>
     }]
   }
   Blend Poses by bool(
    BlendPose_0=Component To Local(ComponentPose=@n1)
    BlendPose_1=Component To Local(ComponentPose=Spring controller - Bone: MSTR_sphereFloat!(ComponentPose:Component Space Pose Link Structure, Alpha:float=0.04) -> (Pose:Component Space Pose Link Structure) {is=title AnimGraphNode_SpringBone, alphaBoolBlend {blendOption=Linear}, springBone {boneName=MSTR_sphereFloat}})
    bActiveValue=Get bIsSkyDiving()
   ) {blendType=Linear}
   Input Pose()
   n2 = "Transform (Modify) Bone - Bone: dyn_sphereFloat"(
    ComponentPose=Local To Component(LocalPose=Use cached pose 'GroundPose'())
    Alpha=1
    Translation=Get crown movement() -> (crown movement:Vector)
    Rotation="0,0,0"
    Scale="1,1,1"
   ) {
    rotationSpace=BCS_WorldSpace
    translationMode=BMM_Additive
    translationSpace=BCS_BoneSpace
    alphaBoolBlend {blendOption=Linear}
    alphaScaleBiasClamp {interpSpeedDecreasing=2, interpSpeedIncreasing=2}
    boneToModify {boneName=dyn_sphereFloat}
   }
   Output Pose(Result=Component To Local(ComponentPose=@n2)) {layerGroup=None}
   AnimGraphNode_StateMachine_1 = Backpack States() {is=title AnimGraphNode_StateMachine}
   Save cached pose 'GroundPose'(Pose=@AnimGraphNode_StateMachine_1) {cacheName=GroundPose}
   AnimGraphNode_StateMachine_0 = Animation Testing() {is=title AnimGraphNode_StateMachine}
  ]}
 Animation Testing {
  entry=State
  path=AnimGraph.AnimGraphNode_StateMachine_0.Animation Testing
  ids=id
  nodes [
   AnimStateNode_0 = State() {
    is=title AnimStateNode
    stateEntered {guid=#1}
    stateFullyBlended {guid=#2}
    stateLeft {guid=#3}
   }
   AnimStateNode_2 = State_1() {
    is=title AnimStateNode
    stateEntered {guid=#4}
    stateFullyBlended {guid=#5}
    stateLeft {guid=#6}
   }
   AnimStateNode_3 = State_2() {
    is=title AnimStateNode
    stateEntered {guid=#7}
    stateFullyBlended {guid=#8}
    stateLeft {guid=#9}
   }
   AnimStateTransitionNode_0 = AnimStateTransitionNode() {bAutomaticRuleBasedOnSequencePlayerInState=true, crossfadeDuration=0}
   AnimStateTransitionNode_2 = AnimStateTransitionNode() {
    transitionEnd {guid=#10}
    transitionInterrupt {guid=#11}
    transitionStart {guid=#12}
   }
   AnimStateTransitionNode_7 = AnimStateTransitionNode()
  ]
  links [
   {
    dst=State_1
    rule=AnimGraph.AnimGraphNode_StateMachine_0.Animation Testing.AnimStateTransitionNode_0.Transition
    src=State
    via=AnimStateTransitionNode_0
   }
   {
    dst=State_2
    rule=AnimGraph.AnimGraphNode_StateMachine_0.Animation Testing.AnimStateTransitionNode_2.Transition
    src=State_1
    via=AnimStateTransitionNode_2
   }
   {
    dst=State
    rule=AnimGraph.AnimGraphNode_StateMachine_0.Animation Testing.AnimStateTransitionNode_7.Transition
    src=State_2
    via=AnimStateTransitionNode_7
   }
  ]
 }
 State {
  path=AnimGraph.AnimGraphNode_StateMachine_0.Animation Testing.AnimStateNode_0.State
  nodes [Output Animation Pose(Result=Sequence Player() {bLoopAnimation=false})]
 }
 State_1 {
  path=AnimGraph.AnimGraphNode_StateMachine_0.Animation Testing.AnimStateNode_2.State_1
  nodes [Output Animation Pose(Result=Sequence Player() {bLoopAnimation=false})]
 }
 State_2 {
  path=AnimGraph.AnimGraphNode_StateMachine_0.Animation Testing.AnimStateNode_3.State_2
  nodes [
   n1 = Sequence Player 'StrontiumSpark_Backpack_IdleBreak_V01'() {
    bLoopAnimation=false
    sequence=/BRCosmetics/Accessories/FORT_Backpacks/Backpack_M_MED_StrontiumSpark/Animation/StrontiumSpark_Backpack_IdleBreak_V01
   }
   Output Animation Pose(Result=@n1)
  ]
 }
 Transition {path=AnimGraph.AnimGraphNode_StateMachine_0.Animation Testing.AnimStateTransitionNode_0.Transition, nodes [Result()]}
 Transition {path=AnimGraph.AnimGraphNode_StateMachine_0.Animation Testing.AnimStateTransitionNode_2.Transition, nodes [Result()]}
 Transition {path=AnimGraph.AnimGraphNode_StateMachine_0.Animation Testing.AnimStateTransitionNode_7.Transition, nodes [Result()]}
 Backpack States {
  entry=Idle
  path=AnimGraph.AnimGraphNode_StateMachine_1.Backpack States
  ids=id
  nodes [
   AnimStateNode_0 = Idle() {stateEntered {guid=#13, notifyName=Entered_RandomLocoBreak}}
   AnimStateNode_2 = Skydiving() {
    is=title AnimStateNode
    stateEntered {guid=#14}
    stateFullyBlended {guid=#15}
    stateLeft {guid=#16}
   }
   AnimStateNode_4 = Elimination() {is=title AnimStateNode}
   AnimStateTransitionNode_10 = AnimStateTransitionNode()
   AnimStateTransitionNode_11 = AnimStateTransitionNode()
   AnimStateTransitionNode_2 = AnimStateTransitionNode()
   AnimStateTransitionNode_5 = AnimStateTransitionNode() {bAutomaticRuleBasedOnSequencePlayerInState=true}
  ]
  links [
   {
    dst=Idle
    rule=AnimGraph.AnimGraphNode_StateMachine_1.Backpack States.AnimStateTransitionNode_5.Transition
    src=Elimination
    via=AnimStateTransitionNode_5
   }
   {
    dst=Elimination
    rule=AnimGraph.AnimGraphNode_StateMachine_1.Backpack States.AnimStateTransitionNode_11.Transition
    src=Idle
    via=AnimStateTransitionNode_11
   }
   {
    dst=Skydiving
    rule=AnimGraph.AnimGraphNode_StateMachine_1.Backpack States.AnimStateTransitionNode_10.Transition
    src=Idle
    via=AnimStateTransitionNode_10
   }
   {
    dst=Idle
    rule=AnimGraph.AnimGraphNode_StateMachine_1.Backpack States.AnimStateTransitionNode_2.Transition
    src=Skydiving
    via=AnimStateTransitionNode_2
   }
  ]
 }
 Idle {
  path=AnimGraph.AnimGraphNode_StateMachine_1.Backpack States.AnimStateNode_0.Idle
  nodes [
   n1 = Random Sequence Player!() -> (Pose:Pose Link Structure) {
    is=title AnimGraphNode_RandomPlayer
    entries [
     {
      chanceToPlay=0.7
      maxLoopCount=0
      maxPlayRate=1
      minLoopCount=0
      minPlayRate=1
      sequence=/BRCosmetics/Accessories/FORT_Backpacks/Backpack_M_MED_StrontiumSpark/Animation/StrontiumSpark_Backpack_Idle
      blendIn {blendOption=Linear, blendTime=0.2, customCurve=null}
     }
     {
      chanceToPlay=0.3
      maxLoopCount=1
      maxPlayRate=1
      minLoopCount=0
      minPlayRate=1
      sequence=&StrontiumSpark_Backpack_IdleBreak_V01
      blendIn {blendOption=Linear, blendTime=0.2, customCurve=null}
     }
     {
      chanceToPlay=0.2
      maxLoopCount=1
      maxPlayRate=1
      minLoopCount=0
      minPlayRate=1
      sequence=/BRCosmetics/Accessories/FORT_Backpacks/Backpack_M_MED_StrontiumSpark/Animation/StrontiumSpark_Backpack_IdleBreak_V02
      blendIn {blendOption=Linear, blendTime=0.2, customCurve=null}
     }
     {
      chanceToPlay=0.5
      maxLoopCount=1
      maxPlayRate=1
      minLoopCount=0
      minPlayRate=1
      sequence=/BRCosmetics/Accessories/FORT_Backpacks/Backpack_M_MED_StrontiumSpark/Animation/StrontiumSpark_Backpack_Elimination
      blendIn {blendOption=Linear, blendTime=0.2, customCurve=null}
     }
    ]
   }
   n2 = Random Sequence Player() {entries [
     {
      chanceToPlay=0.7
      maxLoopCount=0
      maxPlayRate=1
      minLoopCount=0
      minPlayRate=1
      sequence=&StrontiumSpark_Backpack_Idle
      blendIn {blendOption=Linear, blendTime=0.2, customCurve=null}
     }
     {
      chanceToPlay=0.3
      maxLoopCount=0
      maxPlayRate=1
      minLoopCount=0
      minPlayRate=1
      sequence=&StrontiumSpark_Backpack_IdleBreak_V01
      blendIn {blendOption=Linear, blendTime=0.2, customCurve=null}
     }
     {
      chanceToPlay=0.2
      maxLoopCount=0
      maxPlayRate=1
      minLoopCount=0
      minPlayRate=1
      sequence=&StrontiumSpark_Backpack_IdleBreak_V02
      blendIn {blendOption=Linear, blendTime=0.2, customCurve=null}
     }
    ]}
   Output Animation Pose(Result=Blend Poses by bool(BlendPose_0=@n1, BlendPose_1=@n2, bActiveValue=Get bIsInFrontEnd()) {blendType=Linear})
  ]
 }
 Skydiving {
  path=AnimGraph.AnimGraphNode_StateMachine_1.Backpack States.AnimStateNode_2.Skydiving
  nodes [Output Animation Pose(Result=Sequence Player 'StrontiumSpark_Backpack_Skydive'() {sequence=/BRCosmetics/Accessories/FORT_Backpacks/Backpack_M_MED_StrontiumSpark/Animation/StrontiumSpark_Backpack_Skydive})]
 }
 Elimination {
  path=AnimGraph.AnimGraphNode_StateMachine_1.Backpack States.AnimStateNode_4.Elimination
  nodes [
   n1 = Sequence Player 'StrontiumSpark_Backpack_Elimination'() {
    bLoopAnimation=false
    sequence=&StrontiumSpark_Backpack_Elimination
   }
   Output Animation Pose(Result=@n1)
  ]
 }
 Transition {
  path=AnimGraph.AnimGraphNode_StateMachine_1.Backpack States.AnimStateTransitionNode_10.Transition
  nodes [Result(bCanEnterTransition=Get bIsSkyDiving())]
 }
 Transition {
  path=AnimGraph.AnimGraphNode_StateMachine_1.Backpack States.AnimStateTransitionNode_11.Transition
  nodes [Result(bCanEnterTransition=Get bEnemyPlayerEliminated() -> (bEnemyPlayerEliminated:bool))]
 }
 Transition {
  path=AnimGraph.AnimGraphNode_StateMachine_1.Backpack States.AnimStateTransitionNode_2.Transition
  nodes [Result(bCanEnterTransition=NOT Boolean(A=Get bIsSkyDiving()))]
 }
 Transition {path=AnimGraph.AnimGraphNode_StateMachine_1.Backpack States.AnimStateTransitionNode_5.Transition, nodes [Result()]}
 EventGraph {nodes [
   Event BlueprintUpdateAnimation()
   > n1 = Sequence()
   > n2 = Random Vector Function!(
    execute
    CharacterPartAnimInstance:Custom Character Part Anim Instance Object Reference
    SkelMeshComponent:Mesh Component Object Reference
    JointName:Name=None
    GameTimeInSeconds:double=GetGameTimeInSeconds()
    TimeOffset:double=0
    PerlinScale:double=0
    Frequency:Vector="0.9,1.1,0.75"
    Amplitude:Vector="0.3,0.7,0.5"
    Offset:Vector="0,0,5.5"
   ) -> (then, VectorOut:Vector)
   > Set crown movement(crown movement:Vector=@n2) -> (Output_Get:Vector)
   Set Skydive(execute=@n1.then_1, Skydive:double=SelectFloat(A=0.2, B=1, bPickA=Get bIsSkyDiving())) -> (Output_Get:double)
   Event BlueprintPostEvaluateAnimation() {is=title K2Node_Event}
   > Set bEnemyPlayerEliminated(bEnemyPlayerEliminated:bool=false) -> (Output_Get:bool)
   n3 = Event BlueprintProcessAnimInputEvent()
   > Sequence()
   > Branch(Condition=IsNamed(self=@n3.AnimInputEvent, InEventName=aie.EnemyPlayerEliminated))
   > Set bEnemyPlayerEliminated(bEnemyPlayerEliminated=true)
   TryGetPawnOwner()
   Self-Reference()
  ]}
]
variables [
 AllowLocomotionBreak {type=boolean}
 bEnemyPlayerEliminated {type=boolean}
 Choose_LocoBreak_Index {type=integer}
 crown movement {type=Vector, default=<0.1267 -0.532 5.4203>}
 IsLocoBreaking {type=boolean}
 Skydive {type=number, default=1}
]
